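# Sound cue for a single Jones voice-over line from the Season 22 Battle Pass quest dialogue.
# A lone Dialogue Player node plays the line's Dialogue Wave, with the Jones dialogue voice as speaker and no targets.
# The cue routes through the narrator VO sound class, uses the shared Season 22 quest attenuation asset, and boosts its volume.
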SoundCue /BattlePassS22/Audio/VO/Cues/Part_01/JONES_S22_REGS_06_02_020_Dialogue_Cue
root=DialoguePlayer_0
nodes [{
  id=DialoguePlayer_0
  properties {dialogueWaveParameter {
    dialogueWave=/BattlePassS22/Audio/VO/Waves/Jones/JONES_S22_REGS_06_02_020_Dialogue
    context {speaker=/BattlePassS22/Audio/VO/Jones_DV, targets []}
   }}
 }]
properties {
 attenuationSettings=/BattlePassS22/Audio/VO/S22_Quest_Att
 maxDistance=2097152
 soundClassObject=/Game/Sounds/FortSoundClassesAndModes/Fort_VO_Narrator
 volumeMultiplier=2
}
